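; Creative storm beacon device that supplies per-phase overrides for an advanced storm.
; GetStormPhaseOverride replaces incoming phase values with non-negative properties and can set the end location to the beacon's position.
; An end-radius mesh scales to EndRadius, with authority-set visibility requiring that move mode and a positive radius.
; Beacon mesh and loop audio join parent visibility components; audio mutes when hidden, and both meshes get dynamic materials.
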(event EventBeginPlay
  (|CallParentFunction)
  (Visibility|AssignOnVisibilityChanged (Variables|Default|GetVisibleInGameComponent) (EventDispatchers|CreateEvent :!event Custom|OnVisibilityChanged))
  (Visibility|SetVisibilityComponents (Variables|Default|GetEndRadiusVisibleInGameComponent) (Utilities|Array|MakeArray (Variables|Default|GetEndRadiusMeshComponent)) true))

(event EventOnPropertiesChanged
  (|CallParentFunction)
  (Utilities|FlowControl|Sequence
    (:then_0
      (Networking|SwitchHasAuthority
        (:Authority
          (Visibility|SetVisiblityDuringPhase (Variables|Default|GetEndRadiusVisibleInGameComponent) (select (and (== (Variables|ToyProperties|GetPhaseMoves) 2) (> (Variables|ToyProperties|GetEndRadius) 0.0)) "NewEnumerator1" "NewEnumerator0")))))
    (:then_1
      (CallFunction|UpdateEndRadiusMeshWorldScale))))

(event Custom|OnVisibilityChanged
  (params (Visible :type Boolean))
  (Audio|Components|Audio|SetVolumeMultiplier (Variables|Default|GetStorm_Manager_Loop_Cue) (select Visible 1.0 0.0) :!class /Script/Engine.AudioComponent))

(fn ConvertLocalRadiusToWorldRadius
  (params (LocalRadius :type Float%20%28double-precision%29))
  (return (* (* LocalRadius 1e+02) (Class|GetClassDefaults (Variables|Default|GetStormShieldClass)))))

(fn GetFloatOverride
  (params (Value :type Float%20%28double-precision%29) (Default :type Float%20%28double-precision%29))
  (return (select (>= Value 0.0) Value Default)))

(fn GetIntOverride
  (params (Value :type Integer) (Default :type Integer))
  (return (select (>= Value 0) Value Default)))

(fn GetOverrideMeshMaterial
  (params (MeshComp :type Static%20Mesh%20Component%20Object%20Reference) (MatIdx :type Integer))
  (if (Utilities|Equal(Object) (Variables|GetMeshComp) (Variables|Default|GetMesh))
    (return (Variables|Transient|GetMeshMID))
    (elif (Utilities|Equal(Object) (Variables|GetMeshComp) (Variables|Default|GetEndRadiusMeshComponent))
      (return (Variables|Transient|GetEndRadiusMID)))))

(fn GetStormPhaseOverride
  (params (In_WaitTime_9_D7E797634847A3AF6330D081A73B4F3D :type Float%20%28single-precision%29) (In_ShrinkTime_11_920D8BB140C25E1B39CCE69746CAB8CB :type Float%20%28single-precision%29) (In_EndRadius_28_FC298B6F485AEC661DE91B9AB8ACDA81 :type Float%20%28single-precision%29) (In_EndRadiusPercentage_3_CD98C461485235F1FA37AEA8E2858699 :type Float%20%28single-precision%29) (In_Damage_25_F609C9DC4F281BA561EB5C8B652CBE3D :type Integer) (In_bMoves_14_1DA1BC714B6A48F21E3E4EABE1D63392 :type Boolean) (In_ForceDistanceMin_22_DB22FD94466C510F2ACB7CB3694C9690 :type Float%20%28single-precision%29) (In_ForceDistanceMax_24_B7CD745E4B355347C2F906989571146F :type Float%20%28single-precision%29) (In_ForceDistanceMinPercentage_33_F185525A4B7DEB217D04F990787CB987 :type Float%20%28single-precision%29) (In_ForceDistanceMaxPercentage_34_34AA9D254BC85E7AF6CBAA8F770AEF4F :type Float%20%28single-precision%29) (In_EndLocation_37_C916E3E84C96C120E2615BAADA29FEE4 :type Vector%204%20Structure))
  (bind (_returnvalue_x _returnvalue_y _returnvalue_z) (Math|Vector|BreakVector (Transformation|GetActorLocation)))
  (return (Utilities|Struct|MakeSAdvancedStormPhase (CallFunction|GetFloatOverride (Variables|ToyProperties|GetWaitTime) In_WaitTime_9_D7E797634847A3AF6330D081A73B4F3D) (CallFunction|GetFloatOverride (Variables|ToyProperties|GetShrinkTime) In_ShrinkTime_11_920D8BB140C25E1B39CCE69746CAB8CB) (CallFunction|GetFloatOverride (Variables|ToyProperties|GetEndRadius) In_EndRadius_28_FC298B6F485AEC661DE91B9AB8ACDA81) -1.0 (CallFunction|GetIntOverride (Variables|ToyProperties|GetDamage) In_Damage_25_F609C9DC4F281BA561EB5C8B652CBE3D) (Utilities|Select :Index (Variables|ToyProperties|GetPhaseMoves) :Option%200 In_bMoves_14_1DA1BC714B6A48F21E3E4EABE1D63392 :Option%201 true :Option%202 false :Option%203 false) (CallFunction|GetFloatOverride (Variables|ToyProperties|GetForceDistanceMin) In_ForceDistanceMin_22_DB22FD94466C510F2ACB7CB3694C9690) (CallFunction|GetFloatOverride (Variables|ToyProperties|GetForceDistanceMax) In_ForceDistanceMax_24_B7CD745E4B355347C2F906989571146F) -1.0 -1.0 (select (== (Variables|ToyProperties|GetPhaseMoves) 2) (Math|Vector4|MakeVector4 _returnvalue_x _returnvalue_y _returnvalue_z 1.0) In_EndLocation_37_C916E3E84C96C120E2615BAADA29FEE4))))

(fn GetVisibilityComponents
  (Utilities|FlowControl|Sequence
    (:then_0
      (bind _output_get (LocalVariables|SetNewVisibilityComponents (|CallParentFunction)))
      (Utilities|FlowControl|Sequence
        (:then_0
          (Utilities|Array|Add _output_get (Variables|Default|GetStorm_Manager_Loop_Cue)))
        (:then_1
          (Utilities|Array|Add _output_get (Variables|Default|GetMesh)))))
    (:then_1
      (return (LocalVariables|GetNewVisibilityComponents)))))

(fn OnRep_EndRadius
  (CallFunction|UpdateEndRadiusMeshWorldScale))

(fn UpdateEndRadiusMeshWorldScale
  (bind (_returnvalue_x _returnvalue_y _returnvalue_z) (Math|Vector|BreakVector (Transformation|GetWorldScale (Variables|Default|GetEndRadiusMeshComponent))))
  (Transformation|SetWorldScale3D (Variables|Default|GetEndRadiusMeshComponent) (Math|Vector|MakeVector (CallFunction|ConvertLocalRadiustoWorldRadius (Variables|ToyProperties|GetEndRadius)) (CallFunction|ConvertLocalRadiustoWorldRadius (Variables|ToyProperties|GetEndRadius)) _returnvalue_z)))

(fn ConstructionScript
  (|CallParentFunction)
  (bind _returnvalue (Rendering|Material|CreateDynamicMaterialInstance (Variables|Default|GetMesh) 0 0 "None" :!class /Script/Engine.PrimitiveComponent))
  (Variables|Transient|SetMeshMID _returnvalue)
  (bind _returnvalue_2 (Rendering|Material|CreateDynamicMaterialInstance (Variables|Default|GetEndRadiusMeshComponent) 0 0 "None" :!class /Script/Engine.PrimitiveComponent))
  (Variables|Transient|SetEndRadiusMID _returnvalue_2))
