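; BP_StormBeacon is a Creative storm beacon device Blueprint; this graph handles only visibility setup and property-change hooks.
; On BeginPlay it registers its static mesh with the VisibleInGame component through GetVisibilityComponents.
; BeginPlay, world ready, and the ToyOptionsComponent's property changed and loaded events all call OnPropertiesChanged.
; OnPropertiesChanged and the construction script are empty, so no storm behavior is implemented here.
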(event EventBeginPlay
  (Visibility|SetVisibilityComponents (Variables|Default|GetVisibleInGameComponent) (CallFunction|GetVisibilityComponents) true)
  (CallFunction|OnPropertiesChanged))

(event EventOnWorldReady
  (CallFunction|OnPropertiesChanged))

(event OnAnyPropertyChanged(ToyOptionsComponent)
  (CallFunction|OnPropertiesChanged))

(event OnPropertiesLoaded(ToyOptionsComponent)
  (CallFunction|OnPropertiesChanged))

(event Custom|OnPropertiesChanged)

(fn GetVisibilityComponents
  (return (Utilities|Array|MakeArray (Variables|Debug|GetStaticMeshComponent))))

(fn ConstructionScript)
