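# Additive, two-sided material for the lightning effect of the Creative Storm Beacon device.
# The red channel of the lightning texture pans vertically and is multiplied by a Color parameter, purple by default, boosted by a large constant for intense glow.
# The unpanned green channel of the same texture multiplies the result as a static mask, so the scrolling streaks stay within a fixed shape.
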Material /Game/Creative/Devices/StormBeacon/Materials/M_StormBeacon_Lightning
graph {
 properties {blendMode=BLEND_Additive, pixelDepthOffsetMode=PDOM_Legacy, twoSided=true, usedShadingModels=Default Lit}
 nodes [
  n1 = Panner() {speedY=-0.7}
  n2 = VectorParameter() {parameterName=Color, defaultValue=<rgba 0.4076 0 1 1>}
  n3 = Constant() {r=500}
  Multiply(
   A=Multiply(A=TextureSample(UVs=@n1) {texture=/Game/Creative/Devices/StormBeacon/Textures/T_StormBeacon_Lightning_E}.R, B=Multiply(A=@n2.RGB, B=@n3))
   B=TextureSample() {texture=&T_StormBeacon_Lightning_E}.G
  )
 ]
}
